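# Niagara swing trail for the Clay Plug melee pickaxe, built from three ribbon emitters: TrailSide, Trail_Screen and Trail_Top_Smoke.
# Each emitter spawns short-lived anim-trail ribbons between bone or socket pairs on the system's skeletal mesh, nudged by a camera offset.
# Curves drive width and a dark color fade tinted by time of day.
# TrailSide adds a second renderer and a separate speedline color, while the other two randomize a dynamic material parameter.
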NiagaraSystem /BRCosmetics/Weapons/FORT_Melee/Pickaxe_ClayPlug/FX/NS_Pickaxe_Clayplug_Trail
dependencies {
 usedDataInterfaces [&NiagaraDataInterfaceColorCurve, &NiagaraDataInterfaceCurve, &NiagaraDataInterfaceSkeletalMesh]
 usedDynamicInputs []
 usedModules [
  &self:SystemSpawnScript.SystemScriptSource.SystemScriptGraph.NiagaraNodeAssignment_0.SetVariables_#1
  &SystemState
 ]
}
emitters {
 TrailSide {
  emitterUpdateScript [
   &EmitterState(Life Cycle Mode=::NewEnumerator1, Loop Behavior=::NewEnumerator1, Loop Duration=0.2, UseLoopDelay=false)
   &FN_SpawnAnimTrail()
  ]
  particleSpawnScript [
   &FN_InitializeAnimTrailBoneIndecies()
   &FN_InitializeAnimTrail(AnimTrailWidth=0.9, Lifetime=0.125, SkeletalMesh=$System.Skeletal Mesh)
   &CameraOffset(Camera Offset Amount=10)
  ]
  particleUpdateScript [
   &FN_UpdateAnimTrail(SkeletalMesh=$System.Skeletal Mesh)
   &ScaleColor(
    Scale Mode=::NewEnumerator1
    Scale RGBA=&MakeVector4_FromLinearColor(LinearColor=&MultiplyLinearColor(
      Linear Color=&ColorFromCurve(ColorCurve=DataInterface() {propertyValues {
         AlphaCurve {keys [<at 0.2401 4 RCIM_Cubic>,<at 0.9176 0 RCIM_Cubic>]}
         BlueCurve {keys [{arriveTangent=0.0093, interpMode=RCIM_Cubic, time=0.0745, value=0.0853}, <at 0.4886 0.051>, <at 0.7462 0>]}
         GreenCurve {keys [{arriveTangent=-0.0863, interpMode=RCIM_Cubic, time=0.0745, value=0.094}, <at 0.4886 0.051>, <at 0.7462 0>]}
         RedCurve {keys [{arriveTangent=-1.136, interpMode=RCIM_Cubic, time=0.0745, value=0.107}, <at 0.4886 0.051>, <at 0.7462 0>]}
        }})
      Scale Factor=&MakeVector4_FromLinearColor(LinearColor=&FNTimeOfDayColor(
        Night Color=<rgba 0.2292 0.3814 0.9167 1>
        Morning Color=<rgba 0.7552 0.6529 0.7526 1>
        Day Color=<rgba 1 1 1 1>
        Evening Color=<rgba 0.7917 0.6433 0.6226 1>
       ))
     ))
    ScaleRGBA=true
   ) {module=ScaleColor001}
   &ScaleRibbonWidth(Ribbon Width Scale=&FloatFromCurve(FloatCurve=DataInterface() {propertyValues {Curve {keys [
         <at 0 1.1686 RCIM_Cubic>
         {interpMode=RCIM_Cubic, tangentMode=RCTM_User, time=0.2, value=1.12}
         {interpMode=RCIM_Cubic, tangentMode=RCTM_User, time=1, value=1.25}
        ]}}}))
   /SetVariables(Particles.SpeedlineColor:LinearColor=&MultiplyLinearColor(
     Linear Color=&ColorFromCurve(ColorCurve=DataInterface() {propertyValues {
        AlphaCurve {keys [<at 0 1 RCIM_Cubic>,<at 1 0 RCIM_Cubic>]}
        BlueCurve {keys [<at 0 0.2097 RCIM_Cubic>,<at 1 0.1046 RCIM_Cubic>]}
        GreenCurve {keys [<at 0 0.2191 RCIM_Cubic>,<at 1 0.1119 RCIM_Cubic>]}
        RedCurve {keys [<at 0 0.224 RCIM_Cubic>,<at 1 0.117 RCIM_Cubic>]}
       }})
     Scale Factor=&MakeVector4_FromLinearColor(LinearColor=&FNTimeOfDayColor(
       Night Color=<rgba 0.1227 0.3772 1 1>
       Morning Color=<rgba 0.3 0.1864 0.2972 1>
       Day Color=<rgba 1 1 1 1>
       Evening Color=<rgba 0.7 0.3 0.2442 1>
      ))
    )) {isSetParameters=true}
  ]
  rendererClasses [&NiagaraRibbonRendererProperties]
  renderers [
   {rendererClass=&NiagaraRibbonRendererProperties}
   {rendererClass=&NiagaraRibbonRendererProperties, rendererIndex=1}
  ]
 }
 Trail_Screen {
  emitterUpdateScript [
   &EmitterState(Life Cycle Mode=::NewEnumerator1, Loop Behavior=::NewEnumerator1, Loop Duration=0.2, UseLoopDelay=false)
   &FN_SpawnAnimTrail()
  ]
  particleSpawnScript [
   &FN_InitializeAnimTrailBoneIndecies(First Bone or Socket Index=1, Second Bone or Socket Index=0)
   &FN_InitializeAnimTrail(AnimTrailWidth=1.25, Lifetime=0.135, SkeletalMesh=$System.Skeletal Mesh)
   &CameraOffset(Camera Offset Amount=10)
  ]
  particleUpdateScript [
   &FN_UpdateAnimTrail(SkeletalMesh=$System.Skeletal Mesh)
   &ScaleColor(
    Scale Mode=::NewEnumerator1
    Scale RGBA=&MakeVector4_FromLinearColor(LinearColor=&MultiplyLinearColor(
      Linear Color=&ColorFromCurve(ColorCurve=DataInterface() {propertyValues {
         AlphaCurve {keys [<at 0 4 RCIM_Cubic>,<at 0.5676 1>,<at 0.9527 0 RCIM_Cubic>]}
         BlueCurve {keys [{arriveTangent=0.1336, interpMode=RCIM_Cubic, time=0.009, value=0.072}, <at 0.3309 0.051>, <at 0.6471 0>]}
         GreenCurve {keys [{arriveTangent=-0.4141, interpMode=RCIM_Cubic, time=0.009, value=0.0851}, <at 0.3309 0.051>, <at 0.6471 0>]}
         RedCurve {keys [{arriveTangent=-16.267, interpMode=RCIM_Cubic, time=0.009, value=0.107}, <at 0.3309 0.051>, <at 0.6471 0>]}
        }})
      Scale Factor=&MakeVector4_FromLinearColor(LinearColor=&FNTimeOfDayColor(
        Night Color=<rgba 0.2292 0.3814 0.9167 1>
        Morning Color=<rgba 0.7552 0.6529 0.7526 1>
        Day Color=<rgba 1 1 1 1>
        Evening Color=<rgba 0.7917 0.6433 0.6226 1>
       ))
     ))
    ScaleRGBA=true
   ) {module=ScaleColor002}
   &ScaleRibbonWidth(Ribbon Width Scale=&FloatFromCurve(
     FloatCurve=DataInterface() {propertyValues {Curve {keys [
         <at 0 1.4899 RCIM_Cubic>
         {interpMode=RCIM_Cubic, tangentMode=RCTM_User, time=0.2, value=1.1}
         {interpMode=RCIM_Cubic, tangentMode=RCTM_User, time=1, value=1.5}
        ]}}}
     Scale Curve=1.07
    ))
   &DynamicMaterialParameters(
    Index 0 Param 1=&RandomRangeFloat(Minimum=-0.25, Maximum=0.25)
    Index 0 Param 2=1
    Module.Param0WriteEnabled=true
    Module.Param1WriteEnabled=true
    Module.Param2WriteEnabled=false
    Module.Param3WriteEnabled=false
   )
  ]
  rendererClasses [&NiagaraRibbonRendererProperties]
  renderers [{rendererClass=&NiagaraRibbonRendererProperties}]
 }
 Trail_Top_Smoke {
  emitterUpdateScript [
   &EmitterState(Life Cycle Mode=::NewEnumerator1, Loop Behavior=::NewEnumerator1, Loop Duration=0.2, UseLoopDelay=false)
   &FN_SpawnAnimTrail(AnimTrailWidthMode=::FromFirst)
  ]
  particleSpawnScript [
   &FN_InitializeAnimTrailBoneIndecies(First Bone or Socket Index=1, Second Bone or Socket Index=0)
   &FN_InitializeAnimTrail(AnimTrailWidth=0.3, Lifetime=0.125, Position Offset=<0 0 -3>, SkeletalMesh=$System.Skeletal Mesh)
   &CameraOffset(Camera Offset Amount=10)
  ]
  particleUpdateScript [
   &FN_UpdateAnimTrail(SkeletalMesh=$System.Skeletal Mesh)
   &ScaleColor(
    Scale Mode=::NewEnumerator1
    Scale RGBA=&MakeVector4_FromLinearColor(LinearColor=&MultiplyLinearColor(
      Linear Color=&ColorFromCurve(ColorCurve=DataInterface() {propertyValues {
         AlphaCurve {keys [<at 0.2401 4 RCIM_Cubic>,<at 0.9176 0 RCIM_Cubic>]}
         BlueCurve {keys [{arriveTangent=0.0093, interpMode=RCIM_Cubic, time=0, value=0.0957}, <at 0.3803 0.051>, <at 0.7462 0>]}
         GreenCurve {keys [{arriveTangent=-0.0863, interpMode=RCIM_Cubic, time=0, value=0.1014}, <at 0.3803 0.051>, <at 0.7462 0>]}
         RedCurve {keys [{arriveTangent=-1.136, interpMode=RCIM_Cubic, time=0, value=0.11}, <at 0.3803 0.051>, <at 0.7462 0>]}
        }})
      Scale Factor=&MakeVector4_FromLinearColor(LinearColor=&FNTimeOfDayColor(
        Night Color=<rgba 0.2292 0.3814 0.9167 1>
        Morning Color=<rgba 0.7552 0.6529 0.7526 1>
        Day Color=<rgba 1 1 1 1>
        Evening Color=<rgba 0.7917 0.6433 0.6226 1>
       ))
     ))
    ScaleRGBA=true
   ) {module=ScaleColor002}
   &ScaleRibbonWidth(Ribbon Width Scale=&FloatFromCurve(
     FloatCurve=DataInterface() {propertyValues {Curve {keys [
         <at 0 1.4899 RCIM_Cubic>
         {interpMode=RCIM_Cubic, tangentMode=RCTM_User, time=0.2, value=1.1}
         {interpMode=RCIM_Cubic, tangentMode=RCTM_User, time=1, value=1.5}
        ]}}}
     Scale Curve=1.07
    ))
   &DynamicMaterialParameters(
    Index 0 Param 1=&RandomRangeFloat(Minimum=-0.25, Maximum=0.25)
    Index 0 Param 2=1
    Module.Param0WriteEnabled=true
    Module.Param1WriteEnabled=true
    Module.Param2WriteEnabled=false
    Module.Param3WriteEnabled=false
   )
  ]
  rendererClasses [&NiagaraRibbonRendererProperties]
  renderers [{rendererClass=&NiagaraRibbonRendererProperties}]
 }
}
summary {
 systemName=NS_Pickaxe_Clayplug_Trail
 emitters [
  {emitterName=TrailSide, rendererClasses [&NiagaraRibbonRendererProperties]}
  {emitterName=Trail_Screen, rendererClasses [&NiagaraRibbonRendererProperties]}
  {emitterName=Trail_Top_Smoke, rendererClasses [&NiagaraRibbonRendererProperties]}
 ]
 userVariables [
  User.RibbonIDOffset {type=NiagaraInt32}
  User.WeakPointHitFade {type=NiagaraFloat, defaultValue=0, description=""}
 ]
}
